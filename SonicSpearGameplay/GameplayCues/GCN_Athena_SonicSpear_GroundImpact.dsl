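; Gameplay cue burst notify that dresses the ground impact of a Sonic Spear donut drop.
; After the parent burst, it casts the target to the drop ability actor, aligning debris VFX to the instigator's rotation and impact VFX to the ground normal.
; Debris takes the drop's frosting color, a Size derived from the spawned prop's Z scale, and a Golden flag from its gameplay tag.
; Size and Golden also drive the first audio component.
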(fn OnBurst(EitherVFX-preferthisversion)
  (params (MyTarget :type Actor%20Object%20Reference) (Parameters :type Gameplay%20Cue%20Parameters%20Structure%20%28by%20ref%29) (ParticleComponents :type Array%20of%20FXSystem%20Component%20Object%20References) (AudioComponents :type Array%20of%20Audio%20Component%20Object%20References) (BurstCameraShakeInstance :type Camera%20Shake%20Base%20Object%20Reference) (BurstDecalInstance :type Decal%20Actor%20Object%20Reference))
  (bind (_returnvalue_location _returnvalue_rotation) (Math|Transform|BreakTransform (Transformation|GetActorTransform (Ability|GameplayCue|GetInstigatorActor Parameters))))
  (bind _output (Utilities|Array|Get(acopy) AudioComponents))
  (bind _output_2 (Utilities|Array|Get(acopy) ParticleComponents))
  (|CallParentFunction MyTarget Parameters ParticleComponents AudioComponents BurstCameraShakeInstance BurstDecalInstance)
  (Utilities|FlowControl|Sequence
    (:then_0
      (bind _asbga_sonic_spear_main_drop_parent (Utilities|Casting|CastToBGA_SonicSpear_MainDropParent MyTarget))
      (if (Utilities|Array|IsValidIndex ParticleComponents)
        (Transformation|SetWorldRotation _output_2 _returnvalue_rotation false false) ; Set\nNS_SonicSpear_Ground_Impact_Debris\nWorld Rotation to\ndonut rotation
        (bind _coloring_frostingcoloring_15_ed6d36ac450de5d5c052ada0a0a175be (Utilities|Struct|BreakSSonicSpearDonutDropColor (Variables|/SonicSpearGameplay/Gameplay/Drops/BGA_SonicSpear_MainDropParent.BGA_SonicSpear_MainDropParent_C|GetColoring_FrostingColoring_15_ED6D36AC450DE5D5C052ADA0A0A175BE _asbga_sonic_spear_main_drop_parent)))
        (Effects|Components|ParticleSystem|SetColorParameter _output_2 "Frosting" _coloring_frostingcoloring_15_ed6d36ac450de5d5c052ada0a0a175be) ; NS_SonicSpear_Ground_Impact_Debris
        (Transformation|SetWorldRotation (Utilities|Array|Get(acopy) ParticleComponents 1) (Math|Rotator|MakeRotfromZ (.Normal (Ability|GameplayCue|GetGameplayCueEndLocationandNormal _asbga_sonic_spear_main_drop_parent Parameters))) false false) ; Set\nNS_SonicSpear_Ground_Impact\nWorld Rotation to\nImpact Normal from ground
        (bind (_returnvalue_x _returnvalue_y _returnvalue_z) (Math|Vector|BreakVector (Transformation|GetWorldScale (Variables|/Script/FortniteGame.BuildingSMActor|GetStaticMeshComponent (Variables|/SonicSpearGameplay/Gameplay/Drops/BGA_SonicSpear_MainDropParent.BGA_SonicSpear_MainDropParent_C|GetSpawnedProp _asbga_sonic_spear_main_drop_parent)))))
        (Effects|Components|ParticleSystem|SetFloatParameter _output_2 "Size" (- (* _returnvalue_z 2.0) 1.0)) ; Set\nNS_SonicSpear_Ground_Impact\nand\nNS_SonicSpear_Ground_Impact_Debris\nSize to\nSpawned Prop z scale
        (bind _returnvalue (GameplayTags|HasMatchingGameplayTag(Message) _asbga_sonic_spear_main_drop_parent "(TagName=\"SonicSpear.DropType.Golden\")"))
        (bind _output_get (LocalVariables|SetIsGolden _returnvalue))
        (Effects|Components|ParticleSystem|SetBoolParameter _output_2 "Golden" _output_get)
        ;[1 1 1] Set Audio
        (Utilities|IsValid _output
          (:Is%20Valid
            (Audio|Parameter|SetFloatParameter _output "Size" (- (* _returnvalue_z 2.0) 1.0) :!class /Script/AudioExtensions.AudioParameterControllerInterface)
            (Audio|Parameter|SetBooleanParameter _output "Golden" (LocalVariables|GetIsGolden) :!class /Script/AudioExtensions.AudioParameterControllerInterface)))))
    (:then_1
      (return))))
