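; Burst gameplay cue that plays the destruction VFX for a Sonic Spear physics donut drop.
; After calling the parent burst, it casts the target to the drop parent and configures the first spawned particle component from it.
; It passes frosting color and a size built from the donut mesh's Y scale plus a scalable-float offset, with Golden forced off.
; The effect's relative scale is reset to one and its rotation matches the mesh.
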(fn OnBurst(EitherVFX-preferthisversion)
  (params (MyTarget :type Actor%20Object%20Reference) (Parameters :type Gameplay%20Cue%20Parameters%20Structure%20%28by%20ref%29) (ParticleComponents :type Array%20of%20FXSystem%20Component%20Object%20References) (AudioComponents :type Array%20of%20Audio%20Component%20Object%20References) (BurstCameraShakeInstance :type Camera%20Shake%20Base%20Object%20Reference) (BurstDecalInstance :type Decal%20Actor%20Object%20Reference))
  (bind _output (Utilities|Array|Get(acopy) ParticleComponents))
  (|CallParentFunction MyTarget Parameters ParticleComponents AudioComponents BurstCameraShakeInstance BurstDecalInstance)
  (bind _asb_sonic_spear_drop_physics_parent (Utilities|Casting|CastToB_SonicSpear_Drop_Physics_Parent MyTarget))
  (bind _coloring_frostingcoloring_15_ed6d36ac450de5d5c052ada0a0a175be (Utilities|Struct|BreakSSonicSpearDonutDropColor (Variables|/SonicSpearGameplay/Gameplay/Drops/B_SonicSpear_Drop_Physics_Parent.B_SonicSpear_Drop_Physics_Parent_C|GetColoring_FrostingColoring_15_ED6D36AC450DE5D5C052ADA0A0A175BE _asb_sonic_spear_drop_physics_parent)))
  (Effects|Components|ParticleSystem|SetColorParameter _output "Frosting" _coloring_frostingcoloring_15_ed6d36ac450de5d5c052ada0a0a175be)
  (bind (_relativescale3d_x _relativescale3d_y) (Math|Vector|BreakVector (Variables|/Script/Engine.SceneComponent|GetRelativeScale3D_X (Variables|/SonicSpearGameplay/Gameplay/Drops/B_SonicSpear_Drop_Physics_Parent.B_SonicSpear_Drop_Physics_Parent_C|GetDonutMesh _asb_sonic_spear_drop_physics_parent))))
  (Effects|Components|ParticleSystem|SetFloatParameter _output "Size" (+ _relativescale3d_y (ScalableFloat|GetValueatLevel (Variables|Default|GetDonut%20Impact%20Size%20Offset))))
  (Transformation|SetRelativeScale3D _output "1.000000,1.000000,1.000000")
  (Effects|Components|ParticleSystem|SetBoolParameter _output "Golden" false)
  (Transformation|SetWorldRotation _output (Transformation|GetWorldRotation (Variables|/SonicSpearGameplay/Gameplay/Drops/B_SonicSpear_Drop_Physics_Parent.B_SonicSpear_Drop_Physics_Parent_C|GetDonutMesh _asb_sonic_spear_drop_physics_parent)) false false)
  (return))
