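; Gameplay cue notify for Sonic Spear drop spawns that tailors the burst particle effect to the spawned drop.
; On burst, it casts the target to the main drop parent and, if a particle component exists, configures the first one.
; It sets the Golden flag by comparing the drop type to the golden donut tag.
; It derives the Size parameter linearly from the spawned prop mesh's world Z scale.
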(fn OnBurst(EitherVFX-preferthisversion)
  (params (MyTarget :type Actor%20Object%20Reference) (Parameters :type Gameplay%20Cue%20Parameters%20Structure%20%28by%20ref%29) (ParticleComponents :type Array%20of%20FXSystem%20Component%20Object%20References) (AudioComponents :type Array%20of%20Audio%20Component%20Object%20References) (BurstCameraShakeInstance :type Camera%20Shake%20Base%20Object%20Reference) (BurstDecalInstance :type Decal%20Actor%20Object%20Reference))
  (bind _particlecomponents (Variables|GetParticleComponents))
  (bind _output (Utilities|Array|Get(acopy) _particlecomponents))
  (bind _asbga_sonic_spear_main_drop_parent (Utilities|Casting|CastToBGA_SonicSpear_MainDropParent MyTarget))
  (if (Utilities|Array|IsValidIndex (Variables|GetParticleComponents))
    (Effects|Components|ParticleSystem|SetBoolParameter _output "Golden" (== (Variables|Default|GetGolden%20Donut%20Tag) (Variables|/SonicSpearGameplay/Gameplay/Drops/BGA_SonicSpear_MainDropParent.BGA_SonicSpear_MainDropParent_C|GetDrop%20Type _asbga_sonic_spear_main_drop_parent)))
    (bind (_returnvalue_x _returnvalue_y _returnvalue_z) (Math|Vector|BreakVector (Transformation|GetWorldScale (Class|BGASonicSpearMainDropParent|GetSpawnedPropMesh _asbga_sonic_spear_main_drop_parent))))
    (Effects|Components|ParticleSystem|SetFloatParameter _output "Size" (- (* _returnvalue_z 2.0) 1.0))
    (return)))
